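# Light function material that gives lights a moving disco dot pattern, apparently for the Athena dance platform quest interactables.
# A slowly diagonal-panning noise sample, partly desaturated and multiplied by its own green channel, is masked by faster horizontally panning radial dots borrowed from a shuriken texture.
# Tiler and PanSpeed parameters scale both layers' tiling and motion, and a Color parameter multiplies the combined result.
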Material /Game/Athena/Items/QuestInteractables/DancePlatforms/FX/M_Disco_LightFunction
graph {
 properties {materialDomain=MD_LightFunction, pixelDepthOffsetMode=PDOM_Legacy, usedShadingModels=Default Lit}
 nodes [
  n1 = TextureCoordinate() {uTiling=20, vTiling=20}
  n2 = ScalarParameter() {defaultValue=1, parameterName=Tiler}
  n3 = Multiply(A=ScalarParameter() {defaultValue=1, parameterName=PanSpeed}, B=Time())
  n4 = TextureSample(UVs=Panner(Coordinate=Multiply(A=@n1, B=@n2), Time=@n3) {speedX=0.1, speedY=0.05}) {texture=&T_NoisePack_01}
  n5 = Desaturation(None=@n4.RGB, Fraction=Constant() {r=0.5})
  n6 = Panner(Coordinate=Multiply(A=TextureCoordinate() {uTiling=10, vTiling=5}, B=ScalarParameter() {defaultValue=1, parameterName=Tiler}), Time=@n3) {speedX=1}
  Multiply(
   A=Multiply(
    A=Multiply(A=@n5, B=@n4.G)
    B=TextureSample(UVs=@n6) {texture=/Game/Characters/Player/Male/Small/T3_Items/Shuriken/Textures/T_Shuriken_RadialDots}.G
   )
   B=VectorParameter() {parameterName=Color, defaultValue=<rgba 1 1 1 1>}.RGB
  )
 ]
}
